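# Light function material that casts moving cloud shadows for the Athena time-of-day lighting.
# Two scrolling cloud-coverage samples are subtracted and boosted to form large clouds, using MF_CloudLayerCoord for every layer's size and speed.
# Two smaller wisp textures are combined, masked by the first coverage sample and added on top, except at Low quality.
# The result is remapped by contrast parameters, clamped, and used to blend from ShadowAmount to full light.
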Material /Clyde/TimeOfDay/LightFunction/LF_AthenaClouds
graph {
 properties {materialDomain=MD_LightFunction, pixelDepthOffsetMode=PDOM_Legacy, shadingModel=MSM_Unlit, usedShadingModels=Unlit}
 nodes [
  n1 = ScalarParameter() {defaultValue=1, parameterName=ContrastLow}
  n2 = ScalarParameter() {parameterName=ContrastHigh}
  n3 = StaticBool()
  n4 = TextureSample(UVs=MaterialFunctionCall!(
    PreviewInMatEditor=@n3
    Size=ScalarParameter() {defaultValue=1000, parameterName=WispSize01}
    Speed=ScalarParameter() {defaultValue=1.3686, parameterName=WispSpeed01}
    Resolution
   ) {materialFunction=/Game/TimeOfDay/LightFunction/MF_CloudLayerCoord}.UVs) {samplerSource=SSM_Wrap_WorldGroupSettings, samplerType=SAMPLERTYPE_Grayscale, texture=/Game/Environments/World/LightFunctions/T_spiralLF}
  n5 = ScalarParameter() {defaultValue=0.3, parameterName=WispClipping}
  n6 = TextureSample(UVs=MaterialFunctionCall(
    PreviewInMatEditor=@n3
    Size=ScalarParameter() {defaultValue=500, parameterName=WispSize02}
    Speed=ScalarParameter() {defaultValue=1, parameterName=WispSpeed02}
   ) {materialFunction=&MF_CloudLayerCoord}.UVs) {samplerSource=SSM_Wrap_WorldGroupSettings, samplerType=SAMPLERTYPE_Grayscale, texture=/Game/Environments/World/LightFunctions/T_FORT_LF}
  n7 = LinearInterpolate(
   A=ScalarParameter() {parameterName=WispLow}
   B=ScalarParameter() {defaultValue=1, parameterName=WispHigh}
   Alpha=Subtract(A=@n4.R, B=Multiply(A=@n5, B=@n6.R))
  )
  n8 = TextureSample(UVs=MaterialFunctionCall(
    PreviewInMatEditor=@n3
    Size=ScalarParameter() {defaultValue=90000, parameterName=BigCloudSize01}
    Speed=ScalarParameter() {defaultValue=0.5, parameterName=BigCloudSpeed01}
   ) {materialFunction=&MF_CloudLayerCoord}.UVs) {samplerSource=SSM_Wrap_WorldGroupSettings, samplerType=SAMPLERTYPE_LinearGrayscale, texture=/Game/Athena/Prototype/Terrain/T_CloudCoverage}
  n9 = TextureSample(UVs=MaterialFunctionCall(
    PreviewInMatEditor=@n3
    Size=ScalarParameter() {defaultValue=120000, parameterName=BigCloudSize02}
    Speed=ScalarParameter() {defaultValue=0.5, parameterName=BigCloudSpeed02}
   ) {materialFunction=&MF_CloudLayerCoord}.UVs) {samplerSource=SSM_Wrap_WorldGroupSettings, samplerType=SAMPLERTYPE_LinearGrayscale, texture=&T_CloudCoverage}
  n10 = Multiply(A=@n9.R, B=ScalarParameter() {defaultValue=2, parameterName=BigCloudClip})
  n11 = Multiply(A=Saturate(None=Subtract(A=@n8.R, B=@n10)), B=ScalarParameter() {defaultValue=1.2, parameterName=BigCloudBoost}) {constB=1.2}
  n12 = Add(A=Multiply(A=Saturate(None=@n7), B=@n8.R), B=@n11)
  n13 = QualitySwitch(Default=@n12, Low=@n11, High=@n12, Medium=@n12)
  LinearInterpolate(A=ScalarParameter() {parameterName=ShadowAmount}, Alpha=Clamp(None=LinearInterpolate(A=@n1, B=@n2, Alpha=@n13) {constA=-0.5, constB=8}))
 ]
}
