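# Material function that produces scrolling texture coordinates for a cloud layer, located among the TimeOfDay light function assets.
# World position XY is divided by the Size input to form a planar projection.
# A Panner scrolls it along V only, driven by Time multiplied by the Speed input.
# The PreviewInMatEditor static switch divides Size by a constant, tightening the tiling for material editor preview.
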MaterialFunction /Clyde/TimeOfDay/LightFunction/MF_CloudLayerCoord
graph {nodes [
  Constant() {r=128}
  n1 = ComponentMask(None=WorldPosition().XYZ) {g=true, r=true}
  n2 = FunctionInput() {bUsePreviewValueAsDefault=true, inputName=Size, inputType=FunctionInput_Scalar, previewValue=<5000 0 0 1>}
  n3 = StaticSwitch(
   True=Divide(A=@n2, B=Constant() {r=80})
   False=@n2
   Value=FunctionInput(Preview=StaticBool()) {bUsePreviewValueAsDefault=true, inputName=PreviewInMatEditor, inputType=FunctionInput_StaticBool}
  )
  n4 = Time()
  n5 = FunctionInput() {bUsePreviewValueAsDefault=true, inputName=Speed, inputType=FunctionInput_Scalar, sortPriority=1, previewValue=<1 0 0 1>}
  FunctionOutput(None=Panner(Coordinate=Divide(A=@n1, B=@n3), Time=Multiply(A=@n4, B=@n5)) {speedY=0.01})
 ]}
